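; Creative color component that computes a device's HSVA color from a preset map, custom color or minigame team color, with cached team palettes.
; GetAdjustedColor layers brightness, optional channel increments and fade-in onto that base color.
; The server clamps replicated channel adjustments, and OnRep starts a curve-driven blend on clients.
; Tick is enabled for fades and blends, fires UpdateVisualsOnClient for the owner to apply, and switches off when both finish.
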(event EventBeginPlay)

(event Custom|OnDesiredColorChannelAdjustementsChanged
  (if (and (not (Math|Float|NearlyEqual(Float) (Variables|UserOptions|GetColorChangeIntervalOption) 0.0 0.01)) (Variables|V2Features|GetbIsColorIncrementingEnabled))
    (Variables|Transient|SetColorChannelsAdjustmentsCurrentTime 0.0)
    (CallFunction|AnimateColorAdjustment 0.0)
    (Components|Tick|SetComponentTickEnabled true)
    (Variables|Transient|SetbColorChannelAdjustmentsActive true)
    (else
      (Variables|Transient|SetColorChannelsAdjustmentsCurrentTime (Variables|UserOptions|GetColorChangeIntervalOption))
      (CallFunction|AnimateColorAdjustment 0.0))))

(event Custom|FadeToBrightnessOptionValue
  (if (not (Math|Float|NearlyEqual(Float) (Variables|UserOptions|GetColorChangeIntervalOption) 0.0 0.01))
    (CallFunction|SnaptoMinBrightness)
    (Components|Tick|SetComponentTickEnabled true)
    (Variables|Transient|SetbColorChangeIntervalActive true)
    (else
      (CallFunction|SnaptoBrightnessOptionValue))))

(event Custom|SnapToBrightnessOptionValue
  (Variables|Transient|SetColorChangeIntervalCurrentTime (Variables|UserOptions|GetColorChangeIntervalOption))
  (CallFunction|AnimateColorChangeInterval 0.0))

(event Custom|SnapToMinBrightness
  (Variables|Transient|SetColorChangeIntervalCurrentTime 0.0)
  (CallFunction|AnimateColorChangeInterval 0.0))

(event EventTick
  (params (DeltaSeconds :type Float%20%28single-precision%29))
  (Utilities|FlowControl|Sequence
    (:then_0
      (if (Variables|Transient|GetbColorChangeIntervalActive)
        (CallFunction|AnimateColorChangeInterval DeltaSeconds)))
    (:then_1
      (if (Variables|Transient|GetbColorChannelAdjustmentsActive)
        (CallFunction|AnimateColorAdjustment DeltaSeconds)))
    (:then_2
      (Default|CallUpdateVisualsonClient))))

(fn AdjustColorByBrightnessUserOption
  (params (Color :type F%20HSVA%20Structure%20%28by%20ref%29))
  (bind _brightnessoption (Variables|UserOptions|GetBrightnessOption))
  (bind _brightnesscurvefloat (Variables|ColorComponent|GetBrightnessCurveFloat)
    (:then
      (LocalVariables|SetScale (Math|Curves|GetFloatValue _brightnesscurvefloat _brightnessoption)))
    (:else
      (LocalVariables|SetScale _brightnessoption)))
  (Utilities|Struct|SetmembersinFHSVA Color (* (Utilities|Struct|BreakFHSVA Color) (LocalVariables|GetScale))))

(fn AdjustColorByColorByFadeInBrightness
  (params (Color :type F%20HSVA%20Structure%20%28by%20ref%29))
  (if (not (Math|Float|NearlyEqual(Float) (Variables|UserOptions|GetColorChangeIntervalOption) 0.0 0.01))
    (Utilities|Struct|SetmembersinFHSVA Color (* (Utilities|Struct|BreakFHSVA Color) (Variables|Transient|GetFadeInBrightness)))))

(fn AnimateColorAdjustment
  (params (DeltaTime :type Float%20%28double-precision%29))
  (bind (_outcurvefloat _bfinishedanimation) (CallFunction|AnimateCurveFloat DeltaTime (Variables|Transient|GetColorChannelsAdjustmentsCurrentTime) (Variables|UserOptions|GetColorChangeIntervalOption) (Variables|ColorComponent|GetColorAdjustmentCurveFloat)))
  (Variables|Transient|SetCurrentColorChannelAdjustments (Math|Color|Lerp(LinearColor) (Variables|Transient|GetPreviousColorChannelAdjustements) (Variables|Transient|GetDesiredColorChannelAdjustements) _outcurvefloat))
  (if _bfinishedanimation
    (Variables|Transient|SetbColorChannelAdjustmentsActive false)
    (CallFunction|TrytoTurnTickOff)))

(fn AnimateColorChangeInterval
  (params (DeltaTime :type Float%20%28double-precision%29))
  (bind (_outcurvefloat _bfinishedanimation) (CallFunction|AnimateCurveFloat DeltaTime (Variables|Transient|GetColorChangeIntervalCurrentTime) (Variables|UserOptions|GetColorChangeIntervalOption) (Variables|ColorComponent|GetColorChangeIntervalCurveFloat)))
  (Variables|Transient|SetFadeInBrightness _outcurvefloat)
  (if _bfinishedanimation
    (Variables|Transient|SetbColorChangeIntervalActive false)
    (CallFunction|TrytoTurnTickOff)))

(fn AnimateCurveFloat
  (params (DeltaTime :type Float%20%28double-precision%29) (CurrentTime :type Float%20%28double-precision%29%20%28by%20ref%29) (MaxTime :type Float%20%28double-precision%29) (CurveFloat :type Curve%20Float%20Object%20Reference))
  (bind _timealpha (LocalVariables|GetTimeAlpha))
  (bind _maxtime (Variables|GetMaxTime))
  (bind _currenttime (Variables|GetCurrentTime))
  ;[1 1 1] Increment value
  (Variables|SetBy-RefVar CurrentTime (Math|Float|Clamp(Float) (+ DeltaTime CurrentTime) 0.0 MaxTime))
  ;[1 1 1] Test for max value
  (if (Math|Float|NearlyEqual(Float) (Variables|GetCurrentTime) (Variables|GetMaxTime) 0.01)
    (Variables|SetBy-RefVar CurrentTime (Variables|GetMaxTime))
    (LocalVariables|SetbAnimationFinished true))
  ;[1 1 1] Get alpha value
  (if (> _maxtime 0.0)
    (LocalVariables|SetTimeAlpha (/ _currenttime _maxtime))
    (else
      (LocalVariables|SetTimeAlpha (select (Math|Float|NearlyEqual(Float) _currenttime 0.0 0.01) 0.0 1.0))))
  ;[1 1 1] Get curve float value
  (bind _curvefloat (Variables|GetCurveFloat)
    (:then
      (LocalVariables|SetCurveFloatResult (Math|Curves|GetFloatValue _curvefloat _timealpha)))
    (:else
      (LocalVariables|SetCurveFloatResult _timealpha)))
  (return (LocalVariables|GetCurveFloatResult) (LocalVariables|GetbAnimationFinished)))

(fn DecrementColorChannelAdjustments
  (params (ColorChannelAdjustements :type Linear%20Color%20Structure))
  (CallFunction|ModifyColorChannelAdjustments (Math|Color|LinearColor-LinearColor (Variables|Transient|GetDesiredColorChannelAdjustements) ColorChannelAdjustements)))

(fn GetColorKeyasRGB
  (params (Key :type E_CustomizableLight_Color%20Enum%20%28by%20ref%29 :default "NewEnumerator0"))
  (bind _colormap (Variables|ColorComponent|GetColorMap))
  (bind (_value _returnvalue) (Utilities|Map|Find _colormap Key))
  (bind (_rgba_r _rgba_g _rgba_b) (Math|Color|BreakColor (Class|CreativeColorComponentBlueprintLibrary|HSVAToRGBA _value)))
  (if _returnvalue
    (bind _brightness (CallFunction|GetDefaultBrightness))
    (return (* (/ 1.0 _brightness) _rgba_r) (* (/ 1.0 _brightness) _rgba_g) (* (/ 1.0 _brightness) _rgba_b) 1.0)
    (else
      (return "(R=1.000000,G=1.000000,B=1.000000,A=1.000000)"))))

(fn GetAdjustedColor
  (LocalVariables|SetAdjustedColor (CallFunction|GetBaseColor))
  (Utilities|FlowControl|Sequence
    (:then_0
      (CallFunction|AdjustColorbyBrightnessUserOption (LocalVariables|GetAdjustedColor)))
    (:then_1
      (if (Variables|V2Features|GetbIsColorIncrementingEnabled)
        (CallFunction|AdjustColorbyColorIncrementsUserOption (LocalVariables|GetAdjustedColor))))
    (:then_2
      (CallFunction|AdjustColorbyColorbyFadeinBrightness (LocalVariables|GetAdjustedColor)))
    (:then_3
      (return (LocalVariables|GetAdjustedColor)))))

(fn GetBaseColor
  (bind _colormapkeyoption (Variables|UserOptions|GetColorMapKeyOption))
  (bind _colormap (Variables|ColorComponent|GetColorMap))
  (bind (_value _returnvalue) (Utilities|Map|Find _colormap _colormapkeyoption))
  (bind _colormapkeyoption_2 (Variables|UserOptions|GetColorMapKeyOption))
  (if (Utilities|Enum|Equal(Enum) _colormapkeyoption_2 "NewEnumerator1")
    (return (CallFunction|GetTeamColor))
    (elif (Utilities|Enum|Equal(Enum) _colormapkeyoption_2 "NewEnumerator0")
      (return (CallFunction|GetCustomColor))
      (elif _returnvalue
        (return _value)
        (else
          (return 0.0 0.0 1.0 1.0))))))

(fn GetColorByCurveVector
  (params (CurveVector :type Curve%20Vector%20Object%20Reference) (AdjustedColor :type F%20HSVA%20Structure))
  (bind _basecolor (LocalVariables|GetBaseColor))
  (bind _brightness_6_0a16ad3f4c6c520b30f6959130f47e90 (Utilities|Struct|BreakFHSVA _basecolor))
  (bind (_hue_2_285c8ddb4fae69cbf132f4b895363bd4 _alpha_8_4b9b40b44df08409e18eccb32a214085) (Utilities|Struct|BreakFHSVA (Variables|GetAdjustedColor)))
  (bind _curvevector (Variables|GetCurveVector)
    (:then
      (bind (_returnvalue_x _returnvalue_y _returnvalue_z) (Math|Vector|BreakVector (Math|Curves|GetVectorValue _curvevector _hue_2_285c8ddb4fae69cbf132f4b895363bd4)))
      (Variables|SetNewColor_Hue_2_285C8DDB4FAE69CBF132F4B895363BD4 (Utilities|Struct|MakeFHSVA _returnvalue_x _returnvalue_y _returnvalue_z _alpha_8_4b9b40b44df08409e18eccb32a214085))
      ;[0.25 0.25 0.25] Scale brightness
      (LocalVariables|SetBaseColor (CallFunction|GetBaseColor))
      (if (> _brightness_6_0a16ad3f4c6c520b30f6959130f47e90 0.0)
        (LocalVariables|SetScale (/ (Utilities|Struct|BreakFHSVA (Variables|GetAdjustedColor)) _brightness_6_0a16ad3f4c6c520b30f6959130f47e90))
        (else
          (LocalVariables|SetScale 1.0)))
      (Utilities|Struct|SetmembersinFHSVA (LocalVariables|GetNewColor) (* (Utilities|Struct|BreakFHSVA (LocalVariables|GetNewColor)) (LocalVariables|GetScale)))
      ;[0.5 0.25 0.25] Scale saturation
      (bind _defaultsaturation (CallFunction|GetDefaultSaturation))
      (bind _structout (Utilities|Struct|SetmembersinFHSVA (LocalVariables|GetNewColor) (* (Utilities|Struct|BreakFHSVA (LocalVariables|GetNewColor)) (Math|Float|MapRangeUnclamped (Utilities|Struct|BreakFHSVA (Variables|GetAdjustedColor)) 0.0 _defaultsaturation 0.0 1.0))))
      (return _structout))
    (:else
      (return (Variables|GetAdjustedColor)))))

(fn GetCustomColor
  (bind (_customcoloroption_r _customcoloroption_g _customcoloroption_b) (Math|Color|BreakColor (Variables|GetCustomColorOption_R)))
  (bind _brightness (CallFunction|GetDefaultBrightness))
  (return (Class|CreativeColorComponentBlueprintLibrary|RGBAToHSVA (Math|Color|MakeColor (* (/ _customcoloroption_r 255.0) _brightness) (* (/ _customcoloroption_g 255.0) _brightness) (* (/ _customcoloroption_b 255.0) _brightness)))))

(fn GetDefaultBrightness
  ;[0.25 0.25 0.5] The last entry is generally more complete than the first entry
  (bind _values (Utilities|Map|Values (Variables|ColorComponent|GetColorMap)))
  (bind (_output_hue_2_285c8ddb4fae69cbf132f4b895363bd4 _output_saturation_4_96689e7d4094d69c09c353901b909cb7 _output_brightness_6_0a16ad3f4c6c520b30f6959130f47e90) (Utilities|Struct|BreakFHSVA (Utilities|Array|Get(aref) _values (- (Utilities|Array|Length _values) 1))))
  (return (select (> (Utilities|Array|Length _values) 0) _output_brightness_6_0a16ad3f4c6c520b30f6959130f47e90 1.0)))

(fn GetDefaultSaturation
  ;[0.25 0.25 0.5] The last entry is generally more complete than the first entry
  (bind _values (Utilities|Map|Values (Variables|ColorComponent|GetColorMap)))
  (bind (_output_hue_2_285c8ddb4fae69cbf132f4b895363bd4 _output_saturation_4_96689e7d4094d69c09c353901b909cb7) (Utilities|Struct|BreakFHSVA (Utilities|Array|Get(aref) _values (- (Utilities|Array|Length _values) 1))))
  (return (select (> (Utilities|Array|Length _values) 0) _output_saturation_4_96689e7d4094d69c09c353901b909cb7 1.0)))

(fn GetTeamColor
  (bind _teamcolorindex (Variables|Transient|GetTeamColorIndex))
  (bind _returnvalue (Knobs|Team|GetTeamValueasIndex _teamcolorindex))
  (bind _team (Teams|IndextoTeam _returnvalue))
  (bind _teamcolorindex_2 (Variables|Transient|GetTeamColorIndex))
  (bind (_hsvacolor_hue_2_285c8ddb4fae69cbf132f4b895363bd4 _hsvacolor_saturation_4_96689e7d4094d69c09c353901b909cb7 _hsvacolor_brightness_6_0a16ad3f4c6c520b30f6959130f47e90) (Utilities|Struct|BreakFHSVA (Variables|GetHSVAColor_Hue_2_285C8DDB4FAE69CBF132F4B895363BD4)))
  ;[0.5 1 1] Cache the Team Colors if not done already
  (if (not (Utilities|IsArrayNonEmpty (Variables|Default|GetCachedTeamColors)))
    (CallFunction|UpdateCachedTeamColors))
  ;[1 1 1] Get Team Color Index supporting +16 players
  (bind _cachedminigamelogic (Variables|Default|GetCachedMinigameLogic)
    (:then
      (bind _currentminigame (Variables|/Script/FortniteGame.FortMinigameLogicComponent|GetCurrentMinigame _cachedminigamelogic)
        (:then
          (Utilities|IsValid (Minigame|GetMinigameTeamComponent _currentminigame)
            (:Is%20Valid
              (LocalVariables|SetColorIndex (Minigame|GameRules|GetTeamColorIndex (Minigame|GetMinigameTeamComponent _currentminigame) _team)))
            (:Is%20Not%20Valid
              ;[0.8827 0.1953 1 0.25] Editor Logic, ignore minigame, use team value directly.
              (if (Editor|InEditor) :!label _label_3
                (if (Knobs|Team|IsTeamAll _teamcolorindex_2)
                  (LocalVariables|SetColorIndex -1)
                  (else
                    (LocalVariables|SetColorIndex (Math|Integer|Max(Integer) (Math|Integer|%25%28Integer%29 (Knobs|Team|GetTeamValue _teamcolorindex_2) 16)) :!label _label)))
                ;[1 1 0.5] Return the offical team colors
                (LocalVariables|SetHSVAColor (Class|CreativeColorComponentBlueprintLibrary|RGBAToHSVA (Utilities|Array|Get(aref) (Variables|Default|GetCachedTeamColors) (* (Math|Conversions|ToInteger(Boolean) (and (Utilities|Array|IsValidIndex (Variables|Default|GetCachedTeamColors) (Minigame|GameRules|GetTeamColorIndex (Minigame|GetMinigameTeamComponent _currentminigame) _team)) (Knobs|Team|IsValidTeamIndex _teamcolorindex))) (LocalVariables|GetColorIndex)))) :!label _label_2)
                ;[0.25 0.5 0.25] Scale by the default brightness
                (bind _brightness (CallFunction|GetDefaultBrightness))
                (bind _structout (Utilities|Struct|SetmembersinFHSVA (LocalVariables|GetHSVAColor) (* _brightness _hsvacolor_brightness_6_0a16ad3f4c6c520b30f6959130f47e90)))
                (return _structout))
              (goto _label)))
          (goto _label_2))
        (:else
          _)))
    (:else
      _))
  (goto _label_3))

(fn IncrementColorChannelAdjustments
  (params (ColorChannelAdjustements :type Linear%20Color%20Structure))
  (CallFunction|ModifyColorChannelAdjustments (Math|Color|LinearColor+LinearColor ColorChannelAdjustements (Variables|Transient|GetDesiredColorChannelAdjustements))))

(fn ModifyColorChannelAdjustments
  (params (ColorChannelAdjustements :type Linear%20Color%20Structure))
  (bind (_returnvalue_r _returnvalue_g _returnvalue_b) (Math|Color|BreakColor (Math|Color|LinearColor+LinearColor ColorChannelAdjustements (Variables|Transient|GetDesiredColorChannelAdjustements))))
  (Utilities|HasAuthority
    (:Authority
      (Variables|SetDesiredColorChannelAdjustements_R (Math|Color|MakeColor (Utilities|ClamptoRange _returnvalue_r (Variables|ColorComponent|GetRedColorAdjustmentRange)) (Utilities|ClamptoRange _returnvalue_g (Variables|ColorComponent|GetGreenColorAdjustmentRange)) (Utilities|ClamptoRange _returnvalue_b (Variables|ColorComponent|GetBlueColorAdjustmentRange)))))))

(fn OnRep_DesiredColorChannelAdjustements
  (Networking|CanExecuteCosmeticEvents
    (:True
      (Variables|Transient|SetPreviousColorChannelAdjustements (Variables|Transient|GetCurrentColorChannelAdjustments))
      (CallFunction|OnDesiredColorChannelAdjustementsChanged)
      (Default|CallUpdateVisualsonClient))))

(fn ResetColorChannelAdjustments
  (Variables|Transient|SetDesiredColorChannelAdjustements))

(fn ResetToInitalValues
  (Utilities|FlowControl|Sequence
    (:then_0
      ;[0.25 0.25 0.5] TeamColor
      (Variables|Transient|SetTeamColorIndex (Variables|UserOptions|GetInitialTeamColorFilterOption))
      (Utilities|HasAuthority
        (:Authority
          ;[0.5 0.25 0.25] Color Increments / Decrements
          (CallFunction|ResetColorChannelAdjustments))))
    (:then_1
      (Networking|CanExecuteCosmeticEvents
        (:True
          (Default|CallUpdateVisualsonClient))))))

(fn SetInitialValues
  (params (InitialTeamColorFilter :type Integer) (ColorMapKey :type E_CustomizableLight_Color%20Enum :default "NewEnumerator2") (CustomColor :type Linear%20Color%20Structure :default "(R=1.000000,G=1.000000,B=1.000000,A=1.000000)") (BrightnessOption :type Float%20%28double-precision%29 :default 1.0) (ColorChangeIntervalOption :type Float%20%28double-precision%29 :default 2.0) (MinigameLogic :type Fort%20Minigame%20Logic%20Component%20Object%20Reference))
  (Variables|UserOptions|SetInitialTeamColorFilterOption InitialTeamColorFilter)
  (Variables|UserOptions|SetColorMapKeyOption ColorMapKey)
  (Variables|UserOptions|SetCustomColorOption CustomColor)
  (Variables|UserOptions|SetBrightnessOption BrightnessOption)
  (Variables|UserOptions|SetColorChangeIntervalOption ColorChangeIntervalOption)
  (Variables|Default|SetCachedMinigameLogic MinigameLogic)
  (CallFunction|ResettoInitalValues))

(fn SetTeamColorIndex
  (params (TeamColorIndex :type Integer))
  (Networking|CanExecuteCosmeticEvents
    (:True
      (Variables|Transient|SetTeamColorIndex TeamColorIndex)
      (Default|CallUpdateVisualsonClient))))

(fn TryToTurnTickOff
  (Components|Tick|SetComponentTickEnabled (or (Variables|Transient|GetbColorChangeIntervalActive) (Variables|Transient|GetbColorChannelAdjustmentsActive))))

(fn UpdateCachedTeamColors
  (bind (_teamlightcolors _teamselectorcolors) (Class|GetClassDefaults "/Game/Creative/Sets/MilitaryBase/Props/CreativeTeamColors.CreativeTeamColors_C"))
  ;[1 1 1] The team colors don't change, so cache them
  (if (Variables|Default|GetbUseLightColorsForTeamColors)
    (Variables|Default|SetCachedTeamColors _teamlightcolors)
    (else
      (Variables|Default|SetCachedTeamColors _teamselectorcolors))))
