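; Blueprint function library of colour helper functions for the Creative colour component.
; HSVAToRGBA and RGBAToHSVA wrap the engine's HSV conversions, carrying alpha alongside hue, saturation and brightness.
; RGBAToVec4 passes the four colour channels straight through as vector components.
; NearlyEqual(Vector4) compares two Vector4 values component by component within an error tolerance, returning true only when all four match.
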(fn HSVAToRGBA
  (params (HSVA_Hue_2_285C8DDB4FAE69CBF132F4B895363BD4 :type Float%20%28single-precision%29) (HSVA_Saturation_4_96689E7D4094D69C09C353901B909CB7 :type Float%20%28single-precision%29) (HSVA_Brightness_6_0A16AD3F4C6C520B30F6959130F47E90 :type Float%20%28single-precision%29) (HSVA_Alpha_8_4B9B40B44DF08409E18ECCB32A214085 :type Float%20%28single-precision%29))
  (return (Math|Color|HSVtoRGB HSVA_Hue_2_285C8DDB4FAE69CBF132F4B895363BD4 HSVA_Saturation_4_96689E7D4094D69C09C353901B909CB7 HSVA_Brightness_6_0A16AD3F4C6C520B30F6959130F47E90 HSVA_Alpha_8_4B9B40B44DF08409E18ECCB32A214085)))

(fn NearlyEqual(Vector4)
  (params (A :type Vector%204%20Structure) (B :type Vector%204%20Structure) (ErrorTolerance :type Float%20%28double-precision%29 :default 1e-06))
  (bind (_x _y _z _w) (Math|Vector4|BreakVector4 B))
  (bind (_x_2 _y_2 _z_2 _w_2) (Math|Vector4|BreakVector4 A))
  (return (Math|Boolean|ANDBoolean (Math|Float|NearlyEqual(Float) _x_2 _x ErrorTolerance) (Math|Float|NearlyEqual(Float) _y_2 _y ErrorTolerance) (Math|Float|NearlyEqual(Float) _z_2 _z ErrorTolerance) (Math|Float|NearlyEqual(Float) _w_2 _w ErrorTolerance))))

(fn RGBAToHSVA
  (params (RGBA :type Linear%20Color%20Structure))
  (bind (_h _s _v _a) (Math|Color|RGBtoHSV RGBA))
  (return _h _s _v _a))

(fn RGBAToVec4
  (params (RGBA_R :type Float%20%28single-precision%29) (RGBA_G :type Float%20%28single-precision%29) (RGBA_B :type Float%20%28single-precision%29) (RGBA_A :type Float%20%28single-precision%29))
  (return RGBA_R RGBA_G RGBA_B RGBA_A))
